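# Physics asset for the SparkleRemix backpack cosmetic, giving collision and joint limits to its left and right cape bone chains.
# The backpack root carries five overlapping capsules as a tilted column, and each cape bone has one capsule.
# Two mirrored seven-bone chains hang from the root, each bone constrained to the one before it.
# Twist limits are zero, while swing limits are widest at the third joint and narrow toward the tip.
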PhysicsAsset /BRCosmetics/Accessories/FORT_Backpacks/F_MED_SparkleRemix_Backpack/Meshes/F_MED_SparkleRemix_Pack_PhysicsAsset
bodies {
 backpack_root [
  {shapeType=Capsule, length=13.2194, radius=10.4784, shapeName=backpack_root_capsule, center=<0 9.6395 -0.5365>, rotation=<rot 13.6096 -90 90>}
  {shapeType=Capsule, length=9.2786, radius=10.4784, shapeName=backpack_root_capsule, center=<0 7.5979 -12.3147>, rotation=<rot 13.6098 -90 90>}
  {shapeType=Capsule, length=9.2786, radius=10.4784, shapeName=backpack_root_capsule, center=<0 2.9967 -32.6113>, rotation=<rot 13.6098 -90 90>}
  {shapeType=Capsule, length=9.2786, radius=10.4784, shapeName=backpack_root_capsule, center=<0 5.0249 -23.3833>, rotation=<rot 13.6098 -90 90>}
  {shapeType=Capsule, length=9.2786, radius=10.4784, shapeName=backpack_root_capsule, center=<0 2.2405 -43.4395>, rotation=<rot 13.6098 -90 90>}
 ]
 cape_01_l [{shapeType=Capsule, length=1.5636, radius=3.4364, shapeName=cape_01_l_capsule, center=<4.274 0 0>, rotation=<rot 90 0 0>}]
 cape_01_r [{shapeType=Capsule, length=1.5636, radius=3.4364, shapeName=cape_01_r_capsule, center=<4.274 0 0>, rotation=<rot 90 0 0>}]
 cape_02_l [{shapeType=Capsule, length=0.6528, radius=3, shapeName=cape_02_l_capsule, center=<3.4928 0 0>, rotation=<rot 90 0 0>}]
 cape_02_r [{shapeType=Capsule, length=0.6528, radius=3, shapeName=cape_02_r_capsule, center=<3.4928 0 0>, rotation=<rot 90 0 0>}]
 cape_03_l [{shapeType=Capsule, length=4.0894, radius=2.6619, shapeName=cape_03_l_capsule, center=<5.0247 0 0>, rotation=<rot 90 0 0>}]
 cape_03_r [{shapeType=Capsule, length=4.0894, radius=2.6619, shapeName=cape_03_r_capsule, center=<5.0247 0 0>, rotation=<rot 90 0 0>}]
 cape_04_l [{shapeType=Capsule, length=4.043, radius=2.946, shapeName=cape_04_l_capsule, center=<5.1279 0 0>, rotation=<rot 90 0 0>}]
 cape_04_r [{shapeType=Capsule, length=4.043, radius=2.946, shapeName=cape_04_r_capsule, center=<5.1279 0 0>, rotation=<rot 90 0 0>}]
 cape_05_l [{shapeType=Capsule, length=0.3959, radius=3, shapeName=cape_05_l_capsule, center=<3.3307 0 0>, rotation=<rot 90 0 0>}]
 cape_05_r [{shapeType=Capsule, length=0.3959, radius=3, shapeName=cape_05_r_capsule, center=<3.3307 0 0>, rotation=<rot 90 0 0>}]
 cape_06_l [{shapeType=Capsule, length=1.8153, radius=3.1847, shapeName=cape_06_l_capsule, center=<3.8269 0 0>, rotation=<rot 90 0 0>}]
 cape_06_r [{shapeType=Capsule, length=1.8153, radius=3.1847, shapeName=cape_06_r_capsule, center=<3.8269 0 0>, rotation=<rot 90 0 0>}]
 cape_07_l [{shapeType=Capsule, length=1.2785, radius=3.7215, shapeName=cape_07_l_capsule, center=<4.4515 0 0>, rotation=<rot 90 0 0>}]
 cape_07_r [{shapeType=Capsule, length=1.2785, radius=3.7215, shapeName=cape_07_r_capsule, center=<4.4237 0 0>, rotation=<rot 90 0 0>}]
}
constraints [
 {bone1Name=cape_01_l, bone2Name=backpack_root, swing1LimitDegrees=15, swing2LimitDegrees=15, twistLimitDegrees=0}
 {bone1Name=cape_02_l, bone2Name=cape_01_l, swing1LimitDegrees=15, swing2LimitDegrees=5, twistLimitDegrees=0}
 {bone1Name=cape_03_l, bone2Name=cape_02_l, swing1LimitDegrees=45, swing2LimitDegrees=45, twistLimitDegrees=0}
 {bone1Name=cape_04_l, bone2Name=cape_03_l, swing1LimitDegrees=30, swing2LimitDegrees=30, twistLimitDegrees=0}
 {bone1Name=cape_05_l, bone2Name=cape_04_l, swing1LimitDegrees=25, swing2LimitDegrees=25, twistLimitDegrees=0}
 {bone1Name=cape_06_l, bone2Name=cape_05_l, swing1LimitDegrees=20, swing2LimitDegrees=20, twistLimitDegrees=0}
 {bone1Name=cape_07_l, bone2Name=cape_06_l, swing1LimitDegrees=15, swing2LimitDegrees=15, twistLimitDegrees=0}
 {bone1Name=cape_01_r, bone2Name=backpack_root, swing1LimitDegrees=15, swing2LimitDegrees=15, twistLimitDegrees=0}
 {bone1Name=cape_02_r, bone2Name=cape_01_r, swing1LimitDegrees=15, swing2LimitDegrees=5, twistLimitDegrees=0}
 {bone1Name=cape_03_r, bone2Name=cape_02_r, swing1LimitDegrees=45, swing2LimitDegrees=45, twistLimitDegrees=0}
 {bone1Name=cape_04_r, bone2Name=cape_03_r, swing1LimitDegrees=30, swing2LimitDegrees=30, twistLimitDegrees=0}
 {bone1Name=cape_05_r, bone2Name=cape_04_r, swing1LimitDegrees=25, swing2LimitDegrees=25, twistLimitDegrees=0}
 {bone1Name=cape_06_r, bone2Name=cape_05_r, swing1LimitDegrees=20, swing2LimitDegrees=20, twistLimitDegrees=0}
 {bone1Name=cape_07_r, bone2Name=cape_06_r, swing1LimitDegrees=15, swing2LimitDegrees=15, twistLimitDegrees=0}
]
